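# Translucent, lit Niagara mesh-particle material for the EctoCat pickaxe's shard effects.
# Dynamic parameters and particle position drive a clamped world-position-offset bend weighted along UV V, plus a time-based sine wave.
# A Fresnel and UV gradient blends two colors through an emissive adjustment function.
# A powered UV gradient feeds MF_Erosion, borrowed from a skydiving contrail effect, for soft-edged dissolve.
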Material /BRCosmetics/Weapons/FORT_Melee/Pickaxe_EctoCat/FX2/M_EctoCat_Shards
graph {
 properties {bUsedWithNiagaraMeshParticles=true, blendMode=BLEND_Translucent, pixelDepthOffsetMode=PDOM_Legacy, usedShadingModels=Default Lit}
 nodes [
  n1 = DynamicParameter() {desc=Parameters set by Niagara, parameterIndex=1, defaultValue=<rgba 1 1 50 1>, paramNames [WPO UV Scale, "", "", ""]}
  n2 = ParticlePositionWS()
  n3 = Subtract(
   A=MaterialFunctionCall(X=@n1.WPO UV Scale, Y=@n1.WPO UV Scale, Z=ComponentMask(None=@n2) {b=true}) {materialFunction=&MakeFloat3}.Result
   B=@n2
  )
  n4 = ScalarParameter() {defaultValue=5, group="☄WPO", parameterName=Interpolate Clamp}
  n5 = DynamicParameter() {defaultValue=<rgba 1 0.4 1 1>, paramNames [ClampInterpolate, SineAmplitude, SineSpeed, WPO_Wave_Direction]}
  n6 = LinearInterpolate(
   A=Clamp(None=@n3, Min=Multiply(B=@n4) {constA=-1}, Max=@n4) {maxDefault=1000, minDefault=-1000}
   B=Clamp(None=@n3, Min=Multiply(B=@n5.ClampInterpolate) {constA=-1}, Max=@n5.ClampInterpolate) {maxDefault=1000, minDefault=-1000}
   Alpha=ScalarParameter() {group="☄WPO", parameterName=UseDynamicParameterClampInterpolate}
  ) {desc=In case Diving causes issues}
  n7 = OneMinus(None=TextureCoordinate())
  n8 = Multiply(A=Power(Base=ComponentMask(None=@n7) {g=true}, Exp=ScalarParameter() {defaultValue=2, group="☄WPO", parameterName=WPOBendPower}), B=@n1.A)
  n9 = Time()
  n10 = DynamicParameter() {defaultValue=<rgba 1 0.4 1 1>, paramNames [ClampInterpolate, SineAmplitude, SineSpeed, WPO_Wave_Direction]}
  n11 = Sine(None=Add(A=Multiply(A=@n9, B=@n10.SineSpeed), B=TextureCoordinate() {coordinateIndex=1}))
  Add(
   A=StaticSwitchParameter(True=LinearInterpolate(B=@n6, Alpha=@n8), False=Constant()) {defaultValue=true, group=WPO, parameterName=Use_Interpolate WPO}
   B=MaterialFunctionCall(B=Multiply(A=@n11, B=@n10.SineAmplitude)) {materialFunction=&AppendMany}.RGB
  )
  n12 = ScalarParameter() {defaultValue=0.08}
  n13 = TextureCoordinate()
  n14 = Multiply(
   A=Power(Base=Fresnel(ExponentIn=@n12), Exp=ScalarParameter() {defaultValue=0.5, parameterName=Param_1})
   B=Subtract(A=ComponentMask(None=@n13) {g=true}, B=ScalarParameter() {defaultValue=-0.4, parameterName=Param_3}) {constB=0.1}
  )
  MaterialFunctionCall(Emissive=LinearInterpolate(
    A=VectorParameter() {parameterName=Param_1, defaultValue=<rgba 3 0.0853 2.7085 1>}.RGB
    B=VectorParameter() {defaultValue=<rgba 0.3991 0 0.6302 1>}.RGB
    Alpha=Saturate(None=@n14)
   )) {materialFunction=&MF_FX_EmissiveAdjustments}
  n15 = TextureCoordinate()
  n16 = Power(
   Base=Subtract(A=ComponentMask(None=@n15) {g=true}, B=ScalarParameter() {defaultValue=-0.45, parameterName=Param_4})
   Exp=ScalarParameter() {defaultValue=15, parameterName=ErosionPower}
  )
  MaterialFunctionCall!(
   Gradient=@n16
   Erosion Amount=ScalarParameter() {defaultValue=0.55, parameterName=Erosion, sliderMax=1}
   Softness=ScalarParameter() {defaultValue=0.06, parameterName=Softness}
  ) {materialFunction=/BRCosmetics/Athena/Effects/Particles/Skydiving/NS_Contrail_DarkAndLight/Functions/MF_Erosion}
  n17 = WorldPosition() {worldPositionShaderOffset=WPT_ExcludeAllShaderOffsets}
  n18 = PreSkinnedLocalBounds()
  n19 = Divide(
   A=Subtract(A=TransformPosition(None=@n17.XYZ) {transformSourceType=TRANSFORMPOSSOURCE_World, transformType=TRANSFORMPOSSOURCE_Instance}, B=@n18.Min)
   B=@n18.Extents
  )
  OneMinus(None=Saturate(None=ComponentMask(None=@n19) {b=true}))
  Reroute(None=Multiply(A=@n16, B=ScalarParameter() {defaultValue=2, parameterName=Param_2}))
 ]
}
