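# Unlit, two-sided masked material for a Niagara mesh-particle shell effect on the EctoCat pickaxe.
# A tiled inky smoke texture pans along V, and a power of it blends between bright magenta and dark purple.
# The smoke's red channel, shaped by UV-V gradients, forms a mask that breaks up toward low V, turns solid as V rises, then fades through an inverted-V smoothstep.
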Material /BRCosmetics/Weapons/FORT_Melee/Pickaxe_EctoCat/FX2/M_EctoCat_Shell
graph {
 properties {
  bUsedWithNiagaraMeshParticles=true
  blendMode=BLEND_Masked
  pixelDepthOffsetMode=PDOM_Legacy
  previewMesh=/Game/Weapons/FORT_Melee/Pickaxe_EctoCat/FX2/SM_FX_Ectocat_ShellMesh_02
  shadingModel=MSM_Unlit
  twoSided=true
  usedShadingModels=Unlit
 }
 nodes [
  n1 = TextureCoordinate()
  n2 = AppendVector(A=ScalarParameter() {defaultValue=3}, B=ScalarParameter() {defaultValue=0.3, parameterName=Param_1})
  n3 = TextureSample(UVs=Panner(Coordinate=Multiply(A=@n1, B=@n2)) {speedY=0.5}) {texture=&T_Inky_Smoke_Tile}
  LinearInterpolate(
   A=VectorParameter() {parameterName=Param_1, defaultValue=<rgba 2.8 0.0796 2.2595 1>}.RGB
   B=VectorParameter() {defaultValue=<rgba 0.1413 0 0.1441 1>}.RGB
   Alpha=Power(Base=@n3.RGB, Exp=ScalarParameter() {defaultValue=0.1, parameterName=Param_2}) {constExponent=0.02}
  )
  n4 = ComponentMask(None=TextureCoordinate()) {g=true}
  n5 = Power(Base=@n4, Exp=ScalarParameter() {defaultValue=0.2, parameterName=s}) {constExponent=0.1}
  Multiply(
   A=LinearInterpolate(A=Multiply(A=@n3.R, B=@n5), Alpha=SmoothStep(Value=@n4) {constMin=0.2}) {constA=1}
   B=SmoothStep(Value=OneMinus(None=@n4)) {constMin=-0.4}
  )
 ]
}
