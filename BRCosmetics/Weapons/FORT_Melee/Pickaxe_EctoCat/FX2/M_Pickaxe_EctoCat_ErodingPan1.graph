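# Translucent Niagara mesh-particle material for EctoCat pickaxe effects, producing eroding, panning smoke with a pink tint.
# An inky smoke texture pans over tiled UVs, with U shifted oppositely on each V half.
# MF_Erosion erodes it by distance from the V center with V-varying softness, then a U edge mask and DepthFade shape the opacity.
# The color lerp uses the same pink for both inputs, so its texture-driven alpha has no effect.
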Material /BRCosmetics/Weapons/FORT_Melee/Pickaxe_EctoCat/FX2/M_Pickaxe_EctoCat_ErodingPan1
graph {
 properties {bUsedWithNiagaraMeshParticles=true, blendMode=BLEND_Translucent, pixelDepthOffsetMode=PDOM_Legacy, usedShadingModels=Default Lit}
 nodes [
  n1 = TextureCoordinate()
  n2 = Sign(None=Subtract(A=ComponentMask(None=@n1) {g=true}) {constB=0.5})
  n3 = AppendVector(A=Add(A=ComponentMask(None=@n1) {r=true}, B=Multiply(A=@n2) {constB=0.25}), B=ComponentMask(None=@n1) {g=true})
  n4 = AppendVector(A=ScalarParameter() {defaultValue=3, parameterName=TileX}, B=ScalarParameter() {defaultValue=0.7, parameterName=TileY})
  n5 = ScalarParameter() {defaultValue=-0.1, parameterName=Param_2}
  n6 = ScalarParameter() {defaultValue=0.5, parameterName=Param_3}
  n7 = TextureSample(UVs=Panner(Coordinate=Multiply(A=@n3, B=@n4), Speed=AppendVector(A=@n5, B=@n6)) {speedY=0.5}) {texture=&T_Inky_Smoke_Tile}
  n8 = Power(Base=@n7.R, Exp=ScalarParameter() {defaultValue=0.8, parameterName=AlphaPower})
  n9 = TextureCoordinate()
  n10 = Subtract(A=OneMinus(None=ComponentMask(None=@n9) {g=true})) {constB=0.5}
  n11 = Multiply(A=Power(Base=Abs(None=@n10), Exp=ScalarParameter() {defaultValue=2}), B=ScalarParameter() {defaultValue=1, parameterName=Param_1})
  n12 = TextureCoordinate()
  n13 = LinearInterpolate(
   A=ScalarParameter() {defaultValue=0.05, parameterName=ErosionSoftness}
   B=ScalarParameter() {defaultValue=1, parameterName=Param_6}
   Alpha=SmoothStep(Min=ScalarParameter() {defaultValue=0.7, parameterName=Param_7}, Value=ComponentMask(None=@n12) {g=true})
  )
  n14 = ScalarParameter() {parameterName=Param_4}
  n15 = ScalarParameter() {defaultValue=0.1, parameterName=Param_5}
  n16 = Subtract(A=ComponentMask(None=@n9) {r=true}) {constB=0.5}
  n17 = Power(Base=OneMinus(None=Abs(None=@n16))) {constExponent=3}
  DepthFade(
   Opacity=Multiply(
    A=MaterialFunctionCall!(Gradient=@n8, Erosion Amount=@n11, Softness=@n13) {materialFunction=/BRCosmetics/Athena/Effects/Particles/Skydiving/NS_Contrail_DarkAndLight/Functions/MF_Erosion}.Default Range
    B=SmoothStep(Min=@n14, Max=@n15, Value=@n17) {constMax=0.4, constMin=0.1}
   )
   FadeDistance=ScalarParameter() {parameterName=DepthFade}
  )
  n18 = VectorParameter() {parameterName=Param_1, defaultValue=<rgba 0.7083 0.0992 0.5756 1>}
  n19 = Power(Base=@n7.RGB, Exp=ScalarParameter() {defaultValue=1, parameterName=ColorPower}) {constExponent=0.02}
  n20 = OneMinus(None=TextureCoordinate())
  LinearInterpolate(A=@n18.RGB, B=@n18.RGB, Alpha=Multiply(A=Round(None=@n19), B=ComponentMask(None=@n20) {g=true}))
  Multiply()
  VectorParameter() {defaultValue=<rgba 0.9047 0.2415 0.761 1>}
 ]
}
