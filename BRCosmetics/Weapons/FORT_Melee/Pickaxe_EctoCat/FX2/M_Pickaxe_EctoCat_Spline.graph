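# Translucent unlit Niagara mesh-particle material for the EctoCat pickaxe effect, drawing spline shards with a purple-to-magenta gradient along V.
# Opacity rounds panning, power-shaped smoke noise into hard-edged wisps, masked along V, with depth fade except on Forward and Mobile.
# World position offset combines SplineThicken width, sine waves, camera offset and a clamped, switchable bend driven by Niagara dynamic parameters.
# UV1 varies noise, wave phase and width per shard.
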Material /BRCosmetics/Weapons/FORT_Melee/Pickaxe_EctoCat/FX2/M_Pickaxe_EctoCat_Spline
graph {
 properties {
  bUsedWithNiagaraMeshParticles=true
  blendMode=BLEND_Translucent
  pixelDepthOffsetMode=PDOM_Legacy
  shadingModel=MSM_Unlit
  usedShadingModels=Unlit
 }
 nodes [
  n1 = TextureCoordinate()
  LinearInterpolate(
   A=VectorParameter() {parameterName=Param_1, defaultValue=<rgba 0.1276 0.0142 0.5 1>}.RGB
   B=VectorParameter() {defaultValue=<rgba 1.2 0 1 1>}.RGB
   Alpha=Power(Base=ComponentMask(None=@n1) {g=true}, Exp=ScalarParameter() {defaultValue=3, parameterName=Color_Power})
  )
  n2 = Add(A=TextureCoordinate(), B=ComponentMask(None=TextureCoordinate() {coordinateIndex=1}) {desc=add offset per shard, r=true})
  n3 = AppendVector(A=ScalarParameter() {defaultValue=0.2, parameterName=TileX}, B=ScalarParameter() {defaultValue=1, parameterName=TileY})
  n4 = TextureSample(UVs=Panner(Coordinate=Multiply(A=@n2, B=@n3)) {speedY=1}) {texture=/Game/Effects/Fort_Effects/Textures/Tile/T_Inky_Smoke_Tile2}
  n5 = ComponentMask(None=TextureCoordinate()) {g=true}
  n6 = ScalarParameter() {defaultValue=-0.75, parameterName=Param_3}
  n7 = Multiply(A=Abs(None=Add(A=@n5, B=@n6) {constB=-0.5}), B=ScalarParameter() {defaultValue=1, parameterName=Param_2}) {constB=1.5}
  n8 = TextureCoordinate()
  n9 = Abs(None=Add(A=ComponentMask(None=@n8) {g=true}) {constB=-0.5})
  n10 = Multiply(A=Round(None=Power(Base=@n4.R, Exp=@n7)), B=SmoothStep(Value=OneMinus(None=@n9)) {constMin=0.5})
  n11 = DepthFade(Opacity=@n10, FadeDistance=ScalarParameter() {defaultValue=100, parameterName=Param_4})
  NamedRerouteDeclaration(None=ShadingPathSwitch(Default=@n11, Deferred=@n11, Forward=@n10, Mobile=@n10)) {name=Opacity, nodeColor=<rgba 0 0.4609 1 1>}
  n12 = DynamicParameter() {defaultValue=<rgba 1 0.4 1 1>, paramNames [ClampInterpolate, SineAmplitude, SineSpeed, WPO_Wave_Direction]}
  n13 = Multiply(A=Time(), B=@n12.SineSpeed)
  n14 = TextureCoordinate() {coordinateIndex=1}
  n15 = Multiply(A=Sine(None=Add(A=@n13, B=@n14)), B=@n12.SineAmplitude)
  n16 = Add(A=Time(), B=ComponentMask(None=TextureCoordinate() {coordinateIndex=1}) {desc=add offset in time, r=true})
  n17 = Multiply(A=ComponentMask(None=TextureCoordinate()) {g=true}, B=ScalarParameter() {defaultValue=1.5, group="☄WPO", parameterName=WPO_WaveFrequency}) {constB=10}
  n18 = ComponentMask(None=TextureCoordinate()) {g=true}
  n19 = Multiply(
   A=Sine(None=Add(A=@n16, B=@n17))
   B=SmoothStep(Min=ScalarParameter() {defaultValue=0.15, group="☄WPO", parameterName=WPO_WaveGradientInfluence, sliderMax=1}, Value=OneMinus(None=@n18)) {constMin=0.2}
  )
  n20 = ScalarParameter() {defaultValue=10, group="☄WPO", parameterName=WPO_WaveIntensity}
  n21 = VectorParameter() {group="☄WPO", parameterName=WPO_WaveDirection, defaultValue=<rgba 1 1 0.3 1>}
  n22 = Subtract(A=ComponentMask(None=TextureCoordinate() {coordinateIndex=1}) {r=true}) {constB=0.5}
  n23 = Power(Base=OneMinus(None=Abs(None=@n22))) {constExponent=8}
  n24 = Constant() {r=0.3}
  n25 = Add(
   A=Multiply(A=Add(A=@n23, B=@n24) {desc=Minimum}, B=Constant() {r=0.8})
   B=ScalarParameter() {defaultValue=1, group="☄WPO", parameterName=Spline_BaseWidth}
  ) {desc=Adjust the Spline Width Per Shards}
  n26 = ScalarParameter() {defaultValue=0.5, group="☄WPO", parameterName=Spline_TipWidth}
  n27 = DynamicParameter() {desc=Parameters set by Niagara, parameterIndex=1, defaultValue=<rgba 1 1 50 1>, paramNames [WPO UV Scale, "", "", ""]}
  n28 = ParticlePositionWS()
  n29 = Subtract(
   A=MaterialFunctionCall(X=@n27.WPO UV Scale, Y=@n27.WPO UV Scale, Z=ComponentMask(None=@n28) {b=true}) {materialFunction=&MakeFloat3}.Result
   B=@n28
  )
  n30 = ScalarParameter() {defaultValue=5, group="☄WPO", parameterName=Interpolate Clamp}
  n31 = Multiply(B=@n30) {constA=-1}
  n32 = DynamicParameter() {defaultValue=<rgba 1 1 1 1>, paramNames [ClampInterpolate, WingOffset_01, WingOffset_02, WPO_Wave_Direction]}
  n33 = Multiply(B=@n32.ClampInterpolate) {constA=-1}
  n34 = ComponentMask(None=OneMinus(None=TextureCoordinate())) {g=true}
  n35 = ScalarParameter() {defaultValue=1, group="☄WPO", parameterName=WPOBendPower}
  n36 = LinearInterpolate(
   B=LinearInterpolate(
    A=Clamp(None=@n29, Min=@n31, Max=@n30) {maxDefault=1000, minDefault=-1000}
    B=Clamp(None=@n29, Min=@n33, Max=@n32.ClampInterpolate) {maxDefault=1000, minDefault=-1000}
    Alpha=ScalarParameter() {group="☄WPO", parameterName=UseDynamicParameterClampInterpolate}
   ) {desc=In case Diving causes issues}
   Alpha=Multiply(A=Power(Base=@n34, Exp=@n35), B=@n27.A)
  )
  n37 = Constant()
  n38 = Add(
   A=Multiply(
    A=Multiply(A=@n19, B=@n20)
    B=Transform(None=@n21.RGB) {desc=Keeps it system space and does not modify Z, transformSourceType=TRANSFORMSOURCE_Instance}
   )
   B=Add(
    A=MaterialFunctionCall(WidthBase=@n25, WidthTip=@n26) {materialFunction=&SplineThicken}.WorldPositionOffset
    B=StaticSwitchParameter(True=@n36, False=@n37) {defaultValue=true, group=WPO, parameterName=Use_Interpolate WPO}
   ) {constB=0, desc=Spline + Wave}
  ) {desc=+ Wpo Interp}
  n39 = TextureCoordinate()
  n40 = MaterialFunctionCall(Offset Amount=Multiply(A=ComponentMask(None=@n39) {g=true}, B=ScalarParameter() {defaultValue=-20, parameterName=CameraOffsetAmount})) {materialFunction=/Engine/Functions/Engine_MaterialFunctions02/WorldPositionOffset/CameraOffset}
  NamedRerouteDeclaration(None=Add(
    A=MaterialFunctionCall(B=@n15) {materialFunction=&AppendMany}.RGB
    B=Add(A=@n38, B=@n40.Clamped Camera Offset)
   )) {name=WPO, nodeColor=<rgba 0 0.2969 1 1>}
  NamedRerouteUsage()
  NamedRerouteUsage()
 ]
}
