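# Translucent unlit material for the EctoCat pickaxe trail, usable on Niagara ribbons and mesh particles.
# A tiled, UV-swizzled inky smoke texture, offset by dynamic parameter PanX, is eroded along U to blend two colors, tinted by particle color.
# Opacity layers erosion along U with particle-driven Erosion and vertical edge fading.
# Rounding hardens the alpha into stepped shapes before depth fade, length gradient and particle alpha are applied.
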Material /BRCosmetics/Weapons/FORT_Melee/Pickaxe_EctoCat/FX2/M_Pickaxe_EctoCat_Trail
graph {
 properties {
  bUsedWithNiagaraMeshParticles=true
  bUsedWithNiagaraRibbons=true
  blendMode=BLEND_Translucent
  pixelDepthOffsetMode=PDOM_Legacy
  shadingModel=MSM_Unlit
  usedShadingModels=Unlit
 }
 nodes [
  n1 = DynamicParameter() {defaultValue=<rgba 1 0 1 1>, paramNames [PanX,Erosion,Param3,Param4]}
  AppendVector(A=ScalarParameter() {parameterName=SpeedY}, B=@n1.PanX)
  n2 = ComponentMask(None=TextureCoordinate()) {r=true}
  n3 = ScalarParameter() {defaultValue=0.4, parameterName=Param_3}
  n4 = LinearInterpolate(
   A=ScalarParameter() {defaultValue=1, parameterName=Param_6}
   B=ScalarParameter() {defaultValue=-0.7, parameterName=Param_2}
   Alpha=Saturate(None=Power(Base=@n2, Exp=@n3))
  ) {constA=1}
  n5 = VectorParameter() {parameterName=Param_1, defaultValue=<rgba 0.3229 0.0526 0.2692 1>}
  n6 = VectorParameter() {defaultValue=<rgba 4 0.336 3.3739 1>}
  n7 = MaterialFunctionCall(XY=TextureCoordinate()) {materialFunction=&Swizzle}
  n8 = AppendVector(A=ScalarParameter() {defaultValue=1.5, parameterName=TileX}, B=ScalarParameter() {defaultValue=1, parameterName=TileY})
  n9 = Constant()
  n10 = TextureSampleParameter2D(UVs=Add(A=Multiply(A=@n7.YX, B=@n8), B=AppendVector(A=@n9, B=@n1.PanX))) {parameterName=Main Tex, texture=&T_Inky_Smoke_Tile}
  n11 = Power(Base=@n10.RGB, Exp=ScalarParameter() {defaultValue=1.5, parameterName=ColorPower}) {constExponent=0.02}
  n12 = TextureCoordinate()
  n13 = MaterialFunctionCall!(
   Gradient=@n11
   Erosion Amount=Power(Base=ComponentMask(None=@n12) {r=true}, Exp=Constant() {r=1})
   Softness=ScalarParameter() {defaultValue=0.4, parameterName=ColorErosionSoftness}
  ) {materialFunction=/BRCosmetics/Athena/Effects/Particles/Skydiving/NS_Contrail_DarkAndLight/Functions/MF_Erosion}
  n14 = OneMinus(None=TextureCoordinate())
  n15 = Multiply(
   A=@n13.Default Range
   B=Power(Base=ComponentMask(None=@n14) {r=true}, Exp=ScalarParameter() {defaultValue=8, group="🔍Alpha", parameterName=ss})
  )
  Multiply(A=Multiply(A=Saturate(None=@n4), B=LinearInterpolate(A=@n5.RGB, B=@n6.RGB, Alpha=@n15)), B=ParticleColor().RGB)
  n16 = ScalarParameter() {defaultValue=2, group="🔍Alpha", parameterName=AlphaPower}
  n17 = ScalarParameter() {defaultValue=0.4, group="🔍Alpha", parameterName=lll}
  n18 = ScalarParameter() {defaultValue=1, group="🔍Alpha", parameterName=s}
  n19 = ComponentMask(None=TextureCoordinate()) {r=true}
  n20 = TextureCoordinate()
  n21 = Multiply(
   A=MaterialFunctionCall(
    Gradient=Power(Base=@n10.R, Exp=@n16)
    Erosion Amount=SmoothStep(Min=@n17, Max=@n18, Value=@n19)
    Softness=ScalarParameter() {defaultValue=0.04, group="🔍Alpha", parameterName=AlphaErosionSoftness}
   ) {materialFunction=&MF_Erosion}.Default Range
   B=SmoothStep(
    Min=ScalarParameter() {group="🔍Alpha", parameterName=Param_4}
    Max=ScalarParameter() {defaultValue=0.02, group="🔍Alpha", parameterName=Param_5}
    Value=ComponentMask(None=@n20) {r=true}
   )
  )
  n22 = ComponentMask(None=TextureCoordinate()) {r=true}
  n23 = MaterialFunctionCall(
   Gradient=Multiply(A=@n10.R, B=OneMinus(None=@n22))
   Erosion Amount=DynamicParameter() {defaultValue=<rgba 1 0 1 1>, paramNames [PanX,Erosion,Param3,Param4]}.Erosion
   Softness=ScalarParameter() {defaultValue=0.8, group="🔍Alpha", parameterName=OverallErosion_Softness}
  ) {materialFunction=&MF_Erosion}
  n24 = Subtract(A=Multiply(A=Saturate(None=@n21), B=@n23.Increased Range), B=OneMinus(None=@n23.Increased Range))
  n25 = ScalarParameter() {defaultValue=0.2, group="🔍Alpha", parameterName=VerticalEdgeFadeFallOff}
  n26 = Subtract(A=ComponentMask(None=TextureCoordinate()) {g=true}) {constB=0.5}
  n27 = Power(Base=OneMinus(None=Abs(None=@n26)), Exp=ScalarParameter() {defaultValue=1.5, group="🔍Alpha", parameterName=VerticalEdgeFade}) {constExponent=8}
  n28 = Multiply(
   A=Multiply(A=Saturate(None=@n24), B=SmoothStep(Min=@n25, Value=@n27))
   B=ScalarParameter() {defaultValue=10, group="🔍Alpha", parameterName=VerticalEdgeFadeIntensity}
  )
  n29 = Multiply(A=@n28, B=ScalarParameter() {defaultValue=0.7, group="🔍Alpha", parameterName=CoreAlphaIntensity}) {constB=0.5}
  n30 = NamedRerouteUsage()
  n31 = DepthFade(
   Opacity=Round(None=LinearInterpolate(A=@n29, B=@n28, Alpha=@n30) {constA=0.2})
   FadeDistance=ScalarParameter() {defaultValue=2, group="🔍Alpha", parameterName=Alpha_DepthFade}
  )
  n32 = TextureCoordinate()
  n33 = Power(
   Base=OneMinus(None=ComponentMask(None=@n32) {r=true})
   Exp=ScalarParameter() {defaultValue=15, group="🔍Alpha", parameterName=Alpha_GradientPower}
  ) {constExponent=8}
  Multiply(A=Multiply(A=Multiply(A=@n31, B=@n33), B=ScalarParameter() {defaultValue=0.2, parameterName=Alpha_GlobalOpacity}), B=ParticleColor().A)
  NamedRerouteDeclaration(None=@n13.Default Range) {name=SmokeyNoise, nodeColor=<rgba 0 1 0.5313 1>}
  OneMinus(None=@n19)
  Round(None=@n11)
  ScalarParameter() {defaultValue=3, parameterName=SpeedX}
 ]
}
