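# Translucent unlit Niagara sprite material that draws a wiggling fish for the EctoCat pickaxe effects.
# A time-driven sine offsets UVs horizontally, weighted by power and smoothstep gradients, then a composite corrects orientation before sampling the grayscale fish texture.
# The sample feeds an Ecto curve atlas lookup and a particle color multiply.
# Opacity combines particle alpha, texture and GlobalAlpha, with depth fade skipped on forward and mobile paths.
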Material /BRCosmetics/Weapons/FORT_Melee/Pickaxe_EctoCat/FX2/M_Pickaxe_EctoCat_WigglyFish
graph {
 properties {bUsedWithNiagaraSprites=true, blendMode=BLEND_Translucent, pixelDepthOffsetMode=PDOM_Legacy, shadingModel=MSM_Unlit, usedShadingModels=Unlit}
 nodes [
  n1 = Power(Base=ComponentMask(None=TextureCoordinate()) {g=true}, Exp=ScalarParameter() {defaultValue=4, parameterName=Exp})
  n2 = Sine(None=Multiply(A=Time() {period=0.4}) {constB=4})
  n3 = Multiply(A=AppendVector(A=Multiply(A=@n2) {constA=4}, B=Constant()), B=ScalarParameter() {defaultValue=0.3, parameterName=DistortionIntensity})
  n4 = TextureCoordinate()
  n5 = SmoothStep(Value=ComponentMask(None=@n4) {r=true}) {constMin=0.3}
  n6 = OneMinus(None=@n4)
  n7 = Multiply(A=Multiply(A=Power(Base=@n5)) {constB=0.5}, B=SmoothStep(Value=ComponentMask(None=@n6) {g=true}) {constMin=0.5})
  n8 = Multiply(A=Multiply(A=@n3, B=@n7), B=ScalarParameter() {defaultValue=1})
  n9 = Constant()
  n10 = Add(A=Add(A=TextureCoordinate(), B=Multiply(A=@n1, B=@n3)), B=ComponentMask(None=AppendVector(A=@n8, B=@n9)) {g=true, r=true})
  Composite!("XY (V2)"=@n10) {desc=Puts FX back at correct orientation, subgraphName=CorrectOrientation}
  MaterialExpressionReroute_0 = Reroute(None=@n10)
  n11 = MaterialFunctionCall(Float2=MaterialFunctionCall(XY=@MaterialExpressionReroute_0."XY (V2)") {materialFunction=&Swizzle}.YX) {materialFunction=&BreakOutFloat2Components}
  n12 = OneMinus(None=AppendVector(A=@n11.R, B=Multiply(A=@n11.G) {constB=-1}))
  MaterialExpressionReroute_1 = Reroute(None=@n12)
  n13 = TextureSample(UVs=@MaterialExpressionReroute_1.Output) {samplerType=SAMPLERTYPE_Grayscale, texture=&/T_Proto_Fish}
  CurveAtlasRowParameter(CurveTime=@n13.G) {
   atlas=/Game/Weapons/FORT_Melee/Pickaxe_EctoCat/FX2/C_Atlas_Ecto
   curve=/Game/Weapons/FORT_Melee/Pickaxe_EctoCat/FX2/C_Curve_Ectocat
   parameterName=Param_1
  }
  n14 = ParticleColor()
  Multiply(A=@n13.RGB, B=@n14.RGB)
  OneMinus(None=@n7)
  PinBase() {reroutePins ["XY (V2)" {expression=&self:MaterialExpressionReroute_0}]}
  PinBase!(Output=@n12) {reroutePins [Output {expression=&self:MaterialExpressionReroute_1}]}
  n15 = Multiply(A=Multiply(A=@n14.A, B=@n13.R), B=ScalarParameter() {defaultValue=1, parameterName=GlobalAlpha})
  n16 = DepthFade(Opacity=@n15) {fadeDistanceDefault=10}
  ShadingPathSwitch(Default=@n16, Deferred=@n16, Forward=@n15, Mobile=@n15)
 ]
}
