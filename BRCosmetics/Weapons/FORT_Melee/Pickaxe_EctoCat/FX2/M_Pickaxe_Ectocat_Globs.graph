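# Unlit translucent material for purple glob mesh particles in the EctoCat pickaxe's Niagara effects.
# A normalized Fresnel blends two purples, tinted by particle color.
# A panning normal texture scaled by the WPO_Intensity dynamic parameter provides a wobble vector, apparently for vertex offset.
# Depth fade, camera depth fade and particle alpha multiply into a fade, which a shading-path switch replaces with 1 on forward and mobile.
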Material /BRCosmetics/Weapons/FORT_Melee/Pickaxe_EctoCat/FX2/M_Pickaxe_Ectocat_Globs
graph {
 properties {
  bUsedWithNiagaraMeshParticles=true
  blendMode=BLEND_Translucent
  pixelDepthOffsetMode=PDOM_Legacy
  shadingModel=MSM_Unlit
  usedShadingModels=Unlit
 }
 nodes [
  CameraVectorWS()
  n1 = ScalarParameter() {defaultValue=1, parameterName=FresnelThickness}
  n2 = ScalarParameter() {defaultValue=4, parameterName=PowerExponent}
  Multiply(
   A=LinearInterpolate(
    A=VectorParameter() {parameterName=v, defaultValue=<rgba 0.4755 0.0846 0.489 1>}.RGB
    B=VectorParameter() {parameterName=v2, defaultValue=<rgba 0.7274 0.1565 0.8281 1>}.RGB
    Alpha=MaterialFunctionCall!(Thickness=@n1, NormalWS, PowerExponent=@n2) {materialFunction=/Game/Weapons/Prototype/ObjectMover/Materials/MF_NormalizedFresnel}.Power
   )
   B=ParticleColor().RGB
  )
  Multiply(
   A=TextureSample(UVs=Panner() {speedX=0.5, speedY=0.5}) {samplerType=SAMPLERTYPE_Normal, texture=/Game/Global/Textures/Default/Tileable_Patterns/1Normal}.RGB
   B=DynamicParameter() {defaultValue=<rgba 0 1 0 1>, paramNames [Param1, Param2, Erosion, WPO_Intensity]}.WPO_Intensity
  )
  ScreenPosition()
  n3 = ScalarParameter() {defaultValue=15, parameterName=FadeDistance}
  n4 = ScalarParameter() {defaultValue=40, parameterName="Fade Length (S)"}
  n5 = ScalarParameter() {defaultValue=120, parameterName="Fade Offset (S)"}
  n6 = Multiply(
   A=Multiply(
    A=DepthFade(FadeDistance=@n3)
    B=MaterialFunctionCall(Fade Length=@n4, Fade Offset=@n5) {materialFunction=&CameraDepthFade}.Result
   )
   B=ParticleColor().A
  )
  n7 = Constant() {r=1}
  ShadingPathSwitch(Default=@n6, Deferred=@n6, Forward=@n7, Mobile=@n7)
  n8 = TextureCoordinate()
  n9 = ScalarParameter() {defaultValue=0.2, parameterName=B}
  TextureSample(UVs=ComponentMask(None=Multiply(A=@n8, B=@n9) {constB=50}) {g=true, r=true}) {samplerType=SAMPLERTYPE_Grayscale, texture=/FortWater/Textures/LowResBlurredNoise}
 ]
}
