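# Translucent unlit Niagara sprite material for the EctoCat pickaxe's effects, drawing a fish-face texture with a wobbling, color-ramped look.
# Parameters blend between plain, swizzled and doubled-U "mirrored" UVs, which panning swirl noise distorts by a particle dynamic parameter.
# The texture's red channel indexes a curve atlas for color, tinted by particle color.
# Green times particle alpha and GlobalAlpha forms the alpha; a scalar DistortionIntensity sits unused.
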Material /BRCosmetics/Weapons/FORT_Melee/Pickaxe_EctoCat/FX2/M_SpriteEctoCat
graph {
 properties {bUsedWithNiagaraSprites=true, blendMode=BLEND_Translucent, pixelDepthOffsetMode=PDOM_Legacy, shadingModel=MSM_Unlit, usedShadingModels=Unlit}
 nodes [
  n1 = TextureCoordinate()
  n2 = ComponentMask(None=@n1) {r=true}
  n3 = OneMinus(None=ComponentMask(None=@n1) {g=true})
  n4 = LinearInterpolate(
   A=@n1
   B=MaterialFunctionCall(XY=AppendVector(A=@n2, B=@n3)) {materialFunction=&Swizzle}.YX
   Alpha=ScalarParameter() {defaultValue=1, parameterName=UseSwizzle}
  )
  n5 = TextureCoordinate()
  n6 = AppendVector(A=Multiply(A=ComponentMask(None=@n5) {r=true}) {constB=2}, B=ComponentMask(None=@n5) {g=true})
  n7 = ScalarParameter() {defaultValue=1, parameterName=UseMirroredUVs}
  n8 = TextureCoordinate()
  n9 = ScalarParameter() {defaultValue=0.5, parameterName=Param_1}
  n10 = TextureSample(UVs=Panner(Coordinate=Multiply(A=@n8, B=@n9)) {speedY=0.5}) {samplerType=SAMPLERTYPE_Normal, texture=/Game/Gadgets/Assets/VinderTech_GliderChute/Glider_Thrive_Female/FX/Texture/T_SwirlNoise_256}
  n11 = AppendVector(A=@n10.R, B=@n10.R)
  n12 = DynamicParameter() {defaultValue=<rgba 0.04 1 1 1>, paramNames [DistortionIntensity, Param2, Param3, Param4]}
  n13 = TextureSampleParameter2D(UVs=Add(A=LinearInterpolate(A=@n4, B=@n6, Alpha=@n7), B=Multiply(A=@n11, B=@n12.DistortionIntensity))) {parameterName=Tex01, samplerType=SAMPLERTYPE_Grayscale, texture=&/T_FishFaceAlternate}
  n14 = ParticleColor()
  Multiply(
   A=CurveAtlasRowParameter(CurveTime=@n13.R) {atlas=/Game/Weapons/FORT_Melee/Pickaxe_EctoCat/FX2/C_Atlas_Ecto, curve=/Game/Weapons/FORT_Melee/Pickaxe_EctoCat/FX2/C_Curve_Ectocat}
   B=@n14.RGB
  )
  Multiply(A=Multiply(A=@n14.A, B=@n13.G), B=ScalarParameter() {defaultValue=1, parameterName=GlobalAlpha})
  ScalarParameter() {defaultValue=0.04, parameterName=DistortionIntensity}
 ]
}
